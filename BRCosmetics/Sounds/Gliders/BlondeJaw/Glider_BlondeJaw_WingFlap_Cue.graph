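# Sound cue for the wing-flap sound of the BlondeJaw glider cosmetic.
# A Random node picks one of six equally weighted flap recordings and feeds it into a Modulator at the root for pitch and volume variation.
# The cue plays at reduced volume and borrows the shared glider thrust attenuation and concurrency, while routing to the glider landing sound class.
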SoundCue /BRCosmetics/Sounds/Gliders/BlondeJaw/Glider_BlondeJaw_WingFlap_Cue
root=Modulator_0
nodes [
 {id=Modulator_0, children [Random_1]}
 {id=Random_1, children [WavePlayer_9,WavePlayer_8,WavePlayer_4,WavePlayer_3,WavePlayer_2,WavePlayer_0], properties {weights [1,1,1,1,1,1]}}
 {id=WavePlayer_0, properties {soundWaveAssetPtr=&/IP_Glider_BlondeJaw_Flap_06}}
 {id=WavePlayer_2, properties {soundWaveAssetPtr=&/IP_Glider_BlondeJaw_Flap_05}}
 {id=WavePlayer_3, properties {soundWaveAssetPtr=&/IP_Glider_BlondeJaw_Flap_04}}
 {id=WavePlayer_4, properties {soundWaveAssetPtr=&/IP_Glider_BlondeJaw_Flap_03}}
 {id=WavePlayer_8, properties {soundWaveAssetPtr=&/IP_Glider_BlondeJaw_Flap_02}}
 {id=WavePlayer_9, properties {soundWaveAssetPtr=&/IP_Glider_BlondeJaw_Flap_01}}
]
properties {
 attenuationSettings=&Fort_Player_Glider_Thrust_Athena_Att
 maxDistance=12500
 soundClassObject=&Fort_Gliders_Land
 volumeMultiplier=0.4
 concurrencySet [&Athena_Glider_Thrust_Conc]
}
